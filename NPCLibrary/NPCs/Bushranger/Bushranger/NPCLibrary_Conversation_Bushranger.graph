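# Conversation database for the Bushranger NPC, defining the greeting spoken when a player starts talking to it.
# A single speech node picks a default or contextual line based on first-ever conversation, match conversation count, the NPC's CosmeticProperty.Swap tag, or specific player outfits.
# The greeting then branches through link nodes into shared NPC quest and basic-service entries, Season 17 and 22 conversations, and Bushranger's own dialogue entry.
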ConversationDatabase /NPCLibrary/NPCs/Bushranger/Bushranger/NPCLibrary_Conversation_Bushranger
entry_points [{tag=Conversation.Entry.Bushranger, destinations [#1]}]
nodes [
 ConversationEntryPointNode_7 {
  id=#1
  class=ConversationEntryPointNode
  outputs [#2]
  properties {entryTag {tagName=Conversation.Entry.Bushranger}}
 }
 Speech_C_4 {
  id=#2
  class=Speech_C
  outputs [
   #3
   #4
   #5
   #6
   #7
  ]
  properties {
   nodeName=""
   generalConfig {
    defaultMessage="Nightday weird here. Bushy can help!"
    contextualMessages [
     {
      message="Bushybushy forest guardian!"
      contextRequirements [{requirement {class=FortControllerRequirement_IsFirstEverConversationWithNPC}}]
      priority {curve {rowName=Tandem.Speech.Priority.FirstEverInteraction}}
     }
     {
      message="Nightday weird here. Bushy can help!"
      contextRequirements [
       {
        participantId {tagName=Conversation.Participant.NPC}
        requirement {class=FortControllerRequirement_HasOwnedTag, properties {bInverseCondition=true, ownedTag {tagName=CosmeticProperty.Swap}}}
       }
       {requirement {class=FortControllerRequirement_CheckNPCMatchConversationCount, properties {nPCUniqueIdTag {tagName=AISpawnerData.Type.Tandem.Bushranger}}}}
      ]
      priority {curve {rowName=None}}
     }
     {
      message="Dayforest has sunshine! Thank the spritefriends!"
      contextRequirements [
       {
        participantId {tagName=Conversation.Participant.NPC}
        requirement {class=FortControllerRequirement_HasOwnedTag, properties {ownedTag {tagName=CosmeticProperty.Swap}}}
       }
       {requirement {class=FortControllerRequirement_CheckNPCMatchConversationCount, properties {nPCUniqueIdTag {tagName=AISpawnerData.Type.Tandem.Bushranger}}}}
      ]
      priority {curve {rowName=None}}
     }
     {
      message="Bushy protect forest for newfriend Jade!"
      requirementMatchPolicy=RequireAny
      contextRequirements [{requirement {
         class=FortControllerRequirement_HasCID
         properties {allowedCIds [/BRCosmetics/Athena/Items/Cosmetics/Characters/Character_PolishedJade_Mind]}
        }}]
      priority {curve {rowName=Tandem.Speech.Priority.OutfitReaction}}
     }
     {
      message="Bushy won't let demons mess with forest!"
      requirementMatchPolicy=RequireAny
      contextRequirements [{requirement {class=FortControllerRequirement_HasCID, properties {allowedCIds [&Character_CubicVice_Dual]}}}]
      priority {curve {rowName=Tandem.Speech.Priority.OutfitReaction}}
     }
    ]
   }
  }
 }
 ConversationLinkNode_4 {
  id=#7
  class=ConversationLinkNode
  properties {remoteEntryTag {tagName=Conversation.Entry.Bushranger.Dialogue}}
 }
 ConversationLinkNode_3 {
  id=#6
  class=ConversationLinkNode
  properties {remoteEntryTag {tagName=Conversation.Entry.Season22.Narrative.Q12}}
 }
 ConversationLinkNode_0 {id=#5, class=ConversationLinkNode, properties {remoteEntryTag {tagName=Conversation.Entry.BasicServices}}}
 ConversationLinkNode_1 {id=#4, class=ConversationLinkNode, properties {remoteEntryTag {tagName=Conversation.Entry.NPCLibrary.Quests}}}
 ConversationLinkNode_2 {id=#3, class=ConversationLinkNode, properties {remoteEntryTag {tagName=Conversation.Entry.Season17.Week1}}}
]
